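# Control Rig for the face accessory part of the M_MED_UskThumpTophat player body, adjusting the top hat as the head moves.
# A spherical pose reader measures the head bone's orientation against spine_05, with an asymmetric active region and a falloff.
# Its output weights a small rotation and translation offset on the tophat bone, propagated to children.
# The pose reader's 3D debug drawing is left enabled.
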ControlRigBlueprint /BRCosmetics/Characters/Player/Male/Medium/Bodies/M_MED_UskThumpTophat/Meshes/Parts/M_MED_UskThumpTophat_FaceAcc_CtrlRig
graphs [
 RigVMFunctionLibrary {links [], nodes []}
 RigVMModel {nodes [
   RigUnit_BeginExecution = &RigUnit_BeginExecution()
   n1 = &RigUnit_SphericalPoseReader(
    ActiveRegionScaleFactors=(PositiveWidth=1,NegativeWidth=1,PositiveHeight=2,NegativeHeight=2)
    ActiveRegionSize=0.15
    Debug=(bDrawDebug=True,bDraw2D=False,bDrawLocalAxes=False,DebugScale=25,DebugSegments=20,DebugThickness=0.25)
    DriverAxis=(X=1,Y=0,Z=0)
    DriverItem=(Type=Bone,Name="head")
    FalloffRegionScaleFactors=(PositiveWidth=1,NegativeWidth=1,PositiveHeight=1,NegativeHeight=1)
    FalloffSize=0.37
    FlipHeightScaling=False
    FlipWidthScaling=False
    OptionalParentItem=(Type=Bone,Name="spine_05")
    RotationOffset=(X=-45,Y=0,Z=0)
    ExecutePin=@RigUnit_BeginExecution.ExecutePin
   )
   > RigUnit_OffsetTransformForItem = &RigUnit_OffsetTransformForItem(
    Item=(Type=Bone,Name="tophat")
    OffsetTransform=(Rotation=(X=-0.0085,Y=0.0692,Z=0.1216,W=0.9901),Translation=(X=1.2,Y=0.2,Z=0),Scale3D=(X=1,Y=1,Z=1))
    bPropagateToChildren=True
    Weight=@n1.OutputParam
   )
  ]}
]
hierarchy [
 root {loc=<0 0 0>, rot=<0 0 0 1>}
 pelvis {parent=root, loc=<0 -0.7071 93.1191>, rot=<0.0127 0.707 -0.0127 -0.707>}
 spine_01 {parent=pelvis, loc=<6.7494 -0.6388 0>, rot=<0 0 0.034 -0.9994>}
 spine_02 {parent=spine_01, loc=<7.2518 -0.035 0>, rot=<0 0 0.0238 -0.9997>}
 spine_03 {parent=spine_02, loc=<6.0104 0 0>, rot=<0 0 0.0149 -0.9999>}
 spine_04 {parent=spine_03, loc=<8.1024 0 0>, rot=<0 0 -0.0132 -0.9999>}
 spine_05 {parent=spine_04, loc=<10.0146 0 0>, rot=<0 0 -0.0015 -1>}
 neck_01 {parent=spine_05, loc=<11.3138 0.1529 0>, rot=<0 0 -0.2003 -0.9797>}
 neck_02 {parent=neck_01, loc=<4.5627 0 0>, rot=<0 0 0.0504 -0.9987>}
 head {parent=neck_02, loc=<5.3914 0 0>, rot=<0 0 0.115 -0.9934>}
 FACIAL_C_FacialRoot {parent=head, loc=<2.1524 -0.746 -0.0002>, rot=<0.4879 -0.5118 -0.4879 -0.5118>}
 FACIAL_C_Jaw {parent=FACIAL_C_FacialRoot, loc=<-0.0137 1.9794 2.992>, rot=<0.3472 0 0.0005 -0.9378>}
 FACIAL_C_Jawline {parent=FACIAL_C_Jaw, loc=<0.0083 2.5772 7.7055>, rot=<0.2586 0 0 -0.966>}
 FACIAL_R_Jawline {parent=FACIAL_C_FacialRoot, loc=<-4.5593 5.5312 5.5532>, rot=<0.2146 -0.1574 0.056 -0.9623>}
 FACIAL_L_Jawline {parent=FACIAL_C_FacialRoot, loc=<4.5417 5.5115 5.5669>, rot=<0.2147 0.1575 -0.0551 -0.9623>}
 FACIAL_L_UnderChin {parent=FACIAL_C_FacialRoot, loc=<3.513 7.3144 3.9853>, rot=<0.407 -0.3545 0.0847 -0.8376>}
 FACIAL_R_UnderChin {parent=FACIAL_C_FacialRoot, loc=<-3.5194 7.3224 3.9697>, rot=<0.4074 0.3543 -0.0838 -0.8375>}
 FACIAL_C_UnderChin {parent=FACIAL_C_FacialRoot, loc=<0.0015 8.8515 5.5672>, rot=<0.5046 -0.0001 0.0005 -0.8633>}
 FACIAL_R_EyeCornerOuter {parent=FACIAL_C_FacialRoot, loc=<-3.8642 -1.8366 8.8929>, rot=<0.0053 0.3827 -0.0017 -0.9238>}
 FACIAL_R_ForeheadMid {parent=FACIAL_C_FacialRoot, loc=<-2.7023 -3.6944 9.9361>, rot=<-0.0655 0.1759 0.0201 -0.982>}
 FACIAL_L_Masseter {parent=FACIAL_C_FacialRoot, loc=<6.1205 1.4438 5.1711>, rot=<0.0749 0.113 0.0077 -0.9907>}
 FACIAL_R_ForeheadIn {parent=FACIAL_C_FacialRoot, loc=<-1.318 -3.4945 10.5106>, rot=<-0.0552 0.0697 -0.0042 -0.996>}
 FACIAL_L_ForeheadIn {parent=FACIAL_C_FacialRoot, loc=<1.3117 -3.4931 10.5102>, rot=<-0.0552 -0.0694 0.0051 -0.996>}
 FACIAL_L_ForeheadOut {parent=FACIAL_C_FacialRoot, loc=<3.9713 -3.5657 9.0271>, rot=<-0.1162 -0.2558 -0.0773 -0.9566>}
 FACIAL_L_ForeheadMid {parent=FACIAL_C_FacialRoot, loc=<2.6928 -3.6965 9.9295>, rot=<-0.0656 -0.1756 -0.0192 -0.9821>}
 FACIAL_L_CheekOuter {parent=FACIAL_C_FacialRoot, loc=<3.444 0.6936 7.4645>, rot=<0.0601 -0.3763 0.1029 -0.9188>}
 FACIAL_R_CheekOuter {parent=FACIAL_C_FacialRoot, loc=<-3.4443 0.7014 7.4534>, rot=<0.0604 0.3765 -0.102 -0.9188>}
 FACIAL_R_ForeheadOut {parent=FACIAL_C_FacialRoot, loc=<-3.9873 -3.5659 9.0295>, rot=<-0.116 0.2562 0.0782 -0.9564>}
 FACIAL_R_EyesackUpper {parent=FACIAL_C_FacialRoot, loc=<-2.8388 -2.7768 10.2356>, rot=<0.0543 0.0847 0.0645 -0.9928>}
 FACIAL_L_CheekLower {parent=FACIAL_C_FacialRoot, loc=<4.0733 3.6726 6.6125>, rot=<0.0394 -0.4264 0.0605 -0.9016>}
 FACIAL_R_Masseter {parent=FACIAL_C_FacialRoot, loc=<-6.1276 1.4723 5.1665>, rot=<0.0748 -0.1128 -0.0068 -0.9908>}
 FACIAL_C_NoseBridge {parent=FACIAL_C_FacialRoot, loc=<-0.007 -2.4138 11.3924>, rot=<0.0494 0.0001 0.0005 -0.9988>}
 FACIAL_L_Ear {parent=FACIAL_C_FacialRoot, loc=<6.8438 -0.9877 1.1772>, rot=<-0.0208 -0.0007 -0.0257 -0.9995>}
 FACIAL_C_Forehead {parent=FACIAL_C_FacialRoot, loc=<-0.0035 -3.5099 10.6132>, rot=<-0.0379 0.0002 0.0004 -0.9993>}
 FACIAL_R_CheekLower {parent=FACIAL_C_FacialRoot, loc=<-4.0712 3.671 6.6325>, rot=<0.0398 0.4267 -0.0597 -0.9016>}
 FACIAL_R_Ear {parent=FACIAL_C_FacialRoot, loc=<-6.8978 -0.8608 1.0494>, rot=<-0.0208 0.001 0.0266 -0.9994>}
 tophat_base {parent=head, loc=<6.8457 1.7968 0>, rot=<0 0 0.1175 -0.9931>}
 tophat {parent=tophat_base, loc=<0 0 0>, rot=<0 0 0 1>}
]
